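# Translucent unlit material for the motion blur effect of the BrightIon pickaxe, usable on Niagara mesh particles.
# A panned grayscale streak texture blends two glow colors, with a V gradient and Fresnel boosting brightness.
# Particle color and a dynamic Intensity scale the emissive.
# Opacity combines the texture, particle alpha and a V mask shaped by dynamic Power, with depth fade bypassed on Low and Medium quality.
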Material /BRCosmetics/Weapons/FORT_Melee/Pickaxe_BrightIon/FX/M_Pickaxe_BrightIon_MotionBlur
graph {
 properties {
  bUsedWithNiagaraMeshParticles=true
  blendMode=BLEND_Translucent
  pixelDepthOffsetMode=PDOM_Legacy
  shadingModel=MSM_Unlit
  usedShadingModels=Unlit
 }
 nodes [
  n1 = MaterialFunctionCall(Float2=TextureCoordinate()) {materialFunction=&BreakOutFloat2Components}
  n2 = Add(
   A=Multiply(A=@n1.R, B=ScalarParameter() {defaultValue=3, parameterName=UV_Scale_U}) {constB=6}
   B=ScalarParameter() {parameterName=UV_Offset_U}
  )
  n3 = Add(
   A=Multiply(A=@n1.G, B=ScalarParameter() {defaultValue=0.5, parameterName=UV_Scale_V}) {constB=0.45}
   B=ScalarParameter() {defaultValue=0.2, parameterName=UV_Offset_V}
  ) {constB=0.17}
  n4 = ScalarParameter() {defaultValue=5, parameterName=TilingSpeed}
  n5 = Constant()
  n6 = TextureSample(UVs=Panner(Coordinate=AppendVector(A=@n2, B=@n3), Speed=AppendVector(A=@n4, B=@n5)) {speedX=-15}) {samplerType=SAMPLERTYPE_LinearGrayscale, texture=&/T_BrightIon_MotionBlur}
  n7 = Power(Base=@n6.G) {constExponent=1}
  n8 = LinearInterpolate(
   A=VectorParameter() {parameterName=InnerGlowColor_Dark, defaultValue=<rgba 0 1.35 3 1>}.RGB
   B=VectorParameter() {parameterName=InnerGlowColor_Bright, defaultValue=<rgba 0 0.2833 1 1>}.RGB
   Alpha=OneMinus(None=@n7)
  )
  n9 = ComponentMask(None=TextureCoordinate()) {g=true}
  n10 = LinearInterpolate(A=Multiply(A=@n8) {constB=2}, B=@n8, Alpha=Saturate(None=Multiply(A=@n9) {constB=3}))
  n11 = ScalarParameter() {defaultValue=5, parameterName=SideViewFresnelExponent}
  n12 = LinearInterpolate(Alpha=Multiply(A=Fresnel(ExponentIn=@n11) {baseReflectFraction=0}, B=@n7) {constA=1}) {constA=1, constB=20}
  n13 = ParticleColor()
  n14 = DynamicParameter() {paramNames [Power,Intensity,Param3,Param4]}
  MaterialFunctionCall(Emissive=Multiply(A=Multiply(A=@n10, B=@n12), B=Multiply(A=@n13.RGB, B=@n14.Intensity))) {materialFunction=&MF_FX_EmissiveAdjustments}
  n15 = ComponentMask(None=TextureCoordinate()) {g=true}
  n16 = Saturate(None=Add(A=@n15) {constB=-0.2})
  n17 = Saturate(None=Subtract(A=@n15, B=Multiply(A=@n16) {constB=2}))
  n18 = Power(Base=Multiply(A=LinearInterpolate(A=@n15, B=@n17) {constAlpha=0.15}) {constB=2}, Exp=@n14.Power) {constExponent=50}
  n19 = Multiply(A=@n13.A, B=Multiply(A=Multiply(A=@n6.G, B=@n18), B=ScalarParameter() {defaultValue=2, parameterName=Material_Opacity}) {constB=3})
  QualitySwitch(Default=DepthFade(Opacity=@n19, FadeDistance=ScalarParameter() {parameterName=Depthfade}), Low=@n19, Medium=@n19)
 ]
}
