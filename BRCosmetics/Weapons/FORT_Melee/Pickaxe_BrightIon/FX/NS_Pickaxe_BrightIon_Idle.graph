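# Niagara system for the idle effect on the BrightIon pickaxe, a mesh-based sheen from a single emitter, Sheen_Mesh.
# An instantaneous burst spawns mesh particles that do not expire by lifetime and can die when the emitter deactivates, drawn by two mesh renderers.
# Spawn sets a camera offset, dynamic material parameters, a pale blue ColorBladeAlt attribute, and white color with alpha from a front-end/gameplay float modifier.
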NiagaraSystem /BRCosmetics/Weapons/FORT_Melee/Pickaxe_BrightIon/FX/NS_Pickaxe_BrightIon_Idle
dependencies {usedDataInterfaces [], usedDynamicInputs [], usedModules [&SystemState]}
emitters {Sheen_Mesh {
  emitterUpdateScript [
   &EmitterState(
    Inactive Response=::NewEnumerator1
    Life Cycle Mode=::NewEnumerator1
    Loop Behavior=::NewEnumerator1
    Loop Duration Mode=::NewEnumerator1
    UseLoopDelay=false
   )
   &SpawnBurst_Instantaneous(Use Spawn Probability=false, UseLoopCountLimit=false)
  ]
  particleSpawnScript [
   &InitializeParticle(
    Color=&MakeLinearColorFromVectorAndFloat("Vector (RGB)"=<1 1 1>, "Float (Alpha)"=&FN_FrontEndFloatModifier(Gameplay=1, Front End=1))
    Mesh Scale Mode=::NewEnumerator4
    UsePositionOffset=false
    Write Material Random=true
   )
   &DynamicMaterialParameters(
    Index 0 Param 1=1
    Index 0 Param 2=1.1
    Index 1 Param 1=4
    Index 1 Param 2=1.3
    Module.Index0Param0Write=true
    Module.Index0Param1Write=true
    Module.Index0Param2Write=false
    Module.Index0Param3Write=false
    Module.Index1Param0Write=true
    Module.Index1Param1Write=true
    Module.Index1Param2Write=false
    Module.Index1Param3Write=false
    Write Parameter Index 1=true
   )
   &CameraOffset(Camera Offset Amount=1)
   /SetVariables(Particles.ColorBladeAlt:LinearColor=<rgba 0.8 0.85 1 1>) {isSetParameters=true}
  ]
  particleUpdateScript [&ParticleState(Kill Particles When Lifetime Has Elapsed=false, Let Infinitely Lived Particles Die When Emitter Deactivates=true)]
  rendererClasses [&NiagaraMeshRendererProperties]
  renderers [
   {rendererClass=&NiagaraMeshRendererProperties, rendererIndex=0}
   {rendererClass=&NiagaraMeshRendererProperties, rendererIndex=1}
  ]
 }}
summary {
 systemName=NS_Pickaxe_BrightIon_Idle
 emitters [{emitterName=Sheen_Mesh, rendererClasses [&NiagaraMeshRendererProperties]}]
 userVariables [
  User.bIsFrontEnd {type=NiagaraBool}
  User.bIsFrontEndPreview {type=NiagaraBool}
 ]
}
